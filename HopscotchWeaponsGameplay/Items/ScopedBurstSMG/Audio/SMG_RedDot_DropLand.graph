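# Sound cue for the red dot SMG landing when dropped, in the ScopedBurstSMG item's audio.
# A random node picks among three equally weighted landing waves and feeds them through a modulator left at default settings.
# It uses the shared weapon drop attenuation, the loot collect sound class and the stop-oldest weapon drop concurrency set.
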SoundCue /HopscotchWeaponsGameplay/Items/ScopedBurstSMG/Audio/SMG_RedDot_DropLand
root=Modulator_0
nodes [
 {id=Modulator_0, children [Random_1]}
 {id=Random_1, children [WavePlayer_0,WavePlayer_1,WavePlayer_2], properties {bShouldExcludeFromBranchCulling=true, weights [1,1,1]}}
 {id=WavePlayer_0, properties {soundWaveAssetPtr=&/SMG_RedDot_DropLand_01}}
 {id=WavePlayer_1, properties {soundWaveAssetPtr=&/SMG_RedDot_DropLand_02}}
 {id=WavePlayer_2, properties {soundWaveAssetPtr=&/SMG_RedDot_DropLand_03}}
]
properties {
 attenuationSettings=&Fort_Weapon_Drop_Att
 maxDistance=3000
 soundClassObject=&Fort_Loot_Collect
 volumeMultiplier=1
 concurrencyOverrides {resolutionRule=PreventNew}
 concurrencySet [&Weapon_Drop_StopOldest]
}
