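# Fire sound cue for the scoped burst SMG (SMG_RedDot), layering close, medium and distant gunshot variations.
# A mixer sums three branches at decreasing levels, each picking one of four waves at random through a modulator and its own distance attenuation.
# Close shots use the third-person fire sound class with slight pitch variation, while medium and distant use the distant fire class.
# Concurrency sets and overrides cap overlapping shots.
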SoundCue /HopscotchWeaponsGameplay/Items/ScopedBurstSMG/Audio/SMG_RedDot_Fire
root=Mixer_0
nodes [
 {id=Attenuation_2, children [Modulator_2], properties {attenuationSettings=&Fort_Weapon_Projectile_Close_Athena_Att}}
 {
  id=Attenuation_3
  children [Modulator_3]
  properties {attenuationSettings=&Fort_Weapon_Projectile_Medium_Mono_Athena_Att}
 }
 {
  id=Attenuation_4
  children [Modulator_4]
  properties {attenuationSettings=&Fort_Weapon_Projectile_Distant_Mono_Athena_Att}
 }
 {id=Mixer_0, children [SoundClass_2,SoundClass_4,SoundClass_3], properties {inputVolume [1,0.6,0.3]}}
 {id=Modulator_2, children [Random_5], properties {pitchMax=1.1, pitchMin=1, volumeMax=1.2, volumeMin=1.1}}
 {id=Modulator_3, children [Random_6], properties {volumeMax=2, volumeMin=1.9}}
 {id=Modulator_4, children [Random_7], properties {volumeMax=1.95, volumeMin=1.85}}
 {
  id=Random_5
  children [WavePlayer_7, WavePlayer_8, WavePlayer_16, WavePlayer_17]
  properties {bShouldExcludeFromBranchCulling=true, weights [1,1,1,1]}
 }
 {
  id=Random_6
  children [WavePlayer_18, WavePlayer_19, WavePlayer_20, WavePlayer_21]
  properties {bShouldExcludeFromBranchCulling=true, weights [1,1,1,1]}
 }
 {
  id=Random_7
  children [WavePlayer_22, WavePlayer_23, WavePlayer_24, WavePlayer_25]
  properties {bShouldExcludeFromBranchCulling=true, weights [1,1,1,1]}
 }
 {
  id=SoundClass_2
  children [Attenuation_2]
  properties {soundClassOverride=&Fort_WeaponRanged_Fire_3P_PSM}
 }
 {
  id=SoundClass_3
  children [Attenuation_4]
  properties {soundClassOverride=&Fort_WeaponRanged_Fire_Distant}
 }
 {
  id=SoundClass_4
  children [Attenuation_3]
  properties {soundClassOverride=&Fort_WeaponRanged_Fire_Distant}
 }
 {id=WavePlayer_16, properties {soundWaveAssetPtr=&/SMG_RedDot_Fire_Close_03}}
 {id=WavePlayer_17, properties {soundWaveAssetPtr=&/SMG_RedDot_Fire_Close_04}}
 {id=WavePlayer_18, properties {soundWaveAssetPtr=&/SMG_RedDot_Fire_Medium_01}}
 {id=WavePlayer_19, properties {soundWaveAssetPtr=&/SMG_RedDot_Fire_Medium_02}}
 {id=WavePlayer_20, properties {soundWaveAssetPtr=&/SMG_RedDot_Fire_Medium_03}}
 {id=WavePlayer_21, properties {soundWaveAssetPtr=&/SMG_RedDot_Fire_Medium_04}}
 {id=WavePlayer_22, properties {soundWaveAssetPtr=&/SMG_RedDot_Fire_Distant_04}}
 {id=WavePlayer_23, properties {soundWaveAssetPtr=&/SMG_RedDot_Fire_Distant_01}}
 {id=WavePlayer_24, properties {soundWaveAssetPtr=&/SMG_RedDot_Fire_Distant_02}}
 {id=WavePlayer_25, properties {soundWaveAssetPtr=&/SMG_RedDot_Fire_Distant_03}}
 {id=WavePlayer_7, properties {soundWaveAssetPtr=&/SMG_RedDot_Fire_Close_01}}
 {id=WavePlayer_8, properties {soundWaveAssetPtr=&/SMG_RedDot_Fire_Close_02}}
]
properties {
 bDebug=true
 maxDistance=27500
 soundClassObject=&Fort_WeaponRanged_Fire_3P_PSM
 volumeMultiplier=1
 concurrencyOverrides {bLimitToOwner=true, maxCount=3, resolutionRule=StopOldest}
 concurrencySet [&Weapon_Gunfire_Local_Conc, &Weapon_Gunfire_Global_Conc]
}
