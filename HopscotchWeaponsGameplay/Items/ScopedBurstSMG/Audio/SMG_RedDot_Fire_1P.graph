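# First-person fire SoundCue for the red-dot scoped burst SMG, layering the gunshot with quiet bullet casing drops.
# The shot randomly picks one of four waves with slight upward pitch variation and dominates the top mixer.
# Three casing layers play at staggered random delays, each picking from five shared burst-fire SMG drop waves with random downward pitch.
# It uses first-person weapon attenuation, sound class and gunfire concurrency.
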SoundCue /HopscotchWeaponsGameplay/Items/ScopedBurstSMG/Audio/SMG_RedDot_Fire_1P
root=Mixer_1
nodes [
 {id=Delay_0, children [Random_2], properties {delayMax=0.25, delayMin=0.2}}
 {id=Delay_1, children [Random_1], properties {delayMax=0.35, delayMin=0.3}}
 {id=Delay_2, children [Random_3], properties {delayMax=0.45, delayMin=0.4}}
 {id=Mixer_1, children [Modulator_0,Mixer_2], properties {inputVolume [1.2,0.15]}}
 {id=Mixer_2, children [Modulator_1,Modulator_2,Modulator_3], properties {inputVolume [1,1,1]}}
 {id=Modulator_0, children [Random_0], properties {pitchMax=1.02, pitchMin=1}}
 {id=Modulator_1, children [Delay_0], properties {pitchMax=1, pitchMin=0.9}}
 {id=Modulator_2, children [Delay_1], properties {pitchMax=1, pitchMin=0.9}}
 {id=Modulator_3, children [Delay_2], properties {pitchMax=1, pitchMin=0.9}}
 {
  id=Random_0
  children [WavePlayer_0,WavePlayer_1,WavePlayer_2,WavePlayer_3]
  properties {bShouldExcludeFromBranchCulling=true, weights [1,1,1,1]}
 }
 {
  id=Random_1
  children [WavePlayer_4, WavePlayer_5, WavePlayer_6, WavePlayer_7, WavePlayer_13]
  properties {bShouldExcludeFromBranchCulling=true, weights [1,1,1,1,1]}
 }
 {
  id=Random_2
  children [WavePlayer_8, WavePlayer_9, WavePlayer_10, WavePlayer_11, WavePlayer_12]
  properties {bShouldExcludeFromBranchCulling=true, weights [1,1,1,1,1]}
 }
 {
  id=Random_3
  children [WavePlayer_14, WavePlayer_15, WavePlayer_16, WavePlayer_17, WavePlayer_18]
  properties {bShouldExcludeFromBranchCulling=true, weights [1,1,1,1,1]}
 }
 {id=WavePlayer_0, properties {soundWaveAssetPtr=&/SMG_RedDot_Fire_1P_04}}
 {id=WavePlayer_1, properties {soundWaveAssetPtr=&/SMG_RedDot_Fire_1P_01}}
 {id=WavePlayer_10, properties {soundWaveAssetPtr=/Game/Athena/Sounds/Weapons/Pistol_BurstFireSMG/Pistol_BurstFireSMG_Bullet_Casings_Drop_03}}
 {id=WavePlayer_11, properties {soundWaveAssetPtr=/Game/Athena/Sounds/Weapons/Pistol_BurstFireSMG/Pistol_BurstFireSMG_Bullet_Casings_Drop_04}}
 {id=WavePlayer_12, properties {soundWaveAssetPtr=/Game/Athena/Sounds/Weapons/Pistol_BurstFireSMG/Pistol_BurstFireSMG_Bullet_Casings_Drop_05}}
 {id=WavePlayer_13, properties {soundWaveAssetPtr=&Pistol_BurstFireSMG_Bullet_Casings_Drop_05}}
 {id=WavePlayer_14, properties {soundWaveAssetPtr=/Game/Athena/Sounds/Weapons/Pistol_BurstFireSMG/Pistol_BurstFireSMG_Bullet_Casings_Drop_01}}
 {id=WavePlayer_15, properties {soundWaveAssetPtr=/Game/Athena/Sounds/Weapons/Pistol_BurstFireSMG/Pistol_BurstFireSMG_Bullet_Casings_Drop_02}}
 {id=WavePlayer_16, properties {soundWaveAssetPtr=&Pistol_BurstFireSMG_Bullet_Casings_Drop_03}}
 {id=WavePlayer_17, properties {soundWaveAssetPtr=&Pistol_BurstFireSMG_Bullet_Casings_Drop_04}}
 {id=WavePlayer_18, properties {soundWaveAssetPtr=&Pistol_BurstFireSMG_Bullet_Casings_Drop_05}}
 {id=WavePlayer_2, properties {soundWaveAssetPtr=&/SMG_RedDot_Fire_1P_02}}
 {id=WavePlayer_3, properties {soundWaveAssetPtr=&/SMG_RedDot_Fire_1P_03}}
 {id=WavePlayer_4, properties {soundWaveAssetPtr=&Pistol_BurstFireSMG_Bullet_Casings_Drop_01}}
 {id=WavePlayer_5, properties {soundWaveAssetPtr=&Pistol_BurstFireSMG_Bullet_Casings_Drop_02}}
 {id=WavePlayer_6, properties {soundWaveAssetPtr=&Pistol_BurstFireSMG_Bullet_Casings_Drop_03}}
 {id=WavePlayer_7, properties {soundWaveAssetPtr=&Pistol_BurstFireSMG_Bullet_Casings_Drop_04}}
 {id=WavePlayer_8, properties {soundWaveAssetPtr=&Pistol_BurstFireSMG_Bullet_Casings_Drop_01}}
 {id=WavePlayer_9, properties {soundWaveAssetPtr=&Pistol_BurstFireSMG_Bullet_Casings_Drop_02}}
]
properties {
 attenuationSettings=&Fort_Weapon_Projectile_FirstPerson_Athena_Att
 bDebug=true
 maxDistance=2097152
 soundClassObject=&Fort_WeaponRanged_Fire_1P_PSM
 volumeMultiplier=1
 concurrencyOverrides {bLimitToOwner=true, maxCount=21, resolutionRule=StopOldest}
 concurrencySet [&Weapon_Gunfire_FirstPerson_Conc]
}
