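# Sound cue that plays the pickup sound for the red-dot scoped burst SMG.
# A Random node picks one of three equally weighted wave variations and feeds it through a Modulator to the output.
# It uses the weapon pickup attenuation and loot-collect sound class, and its concurrency override allows one instance at a time, blocking new plays while one is active.
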SoundCue /HopscotchWeaponsGameplay/Items/ScopedBurstSMG/Audio/SMG_RedDot_Pickup
root=Modulator_0
nodes [
 {id=Modulator_0, children [Random_0]}
 {id=Random_0, children [WavePlayer_0,WavePlayer_1,WavePlayer_3], properties {bShouldExcludeFromBranchCulling=true, weights [1,1,1]}}
 {id=WavePlayer_0, properties {soundWaveAssetPtr=&/SMG_RedDot_Pickup_01}}
 {id=WavePlayer_1, properties {soundWaveAssetPtr=&/SMG_RedDot_Pickup_02}}
 {id=WavePlayer_3, properties {soundWaveAssetPtr=&/SMG_RedDot_Pickup_03}}
]
properties {
 attenuationSettings=&Fort_Weapon_Pickup_Att
 bOverrideConcurrency=true
 maxDistance=3500
 soundClassObject=&Fort_Loot_Collect
 volumeMultiplier=1.2
 concurrencyOverrides {maxCount=1, resolutionRule=PreventNew}
}
