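# Sound cue for the cartridge reload foley of the red-dot scoped burst SMG in the Hopscotch weapons plugin.
# A random node picks among three equally weighted wave variations and is excluded from branch culling.
# The cue plays through the weapon foley sound class with shared weapon reload attenuation and a boosted volume.
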SoundCue /HopscotchWeaponsGameplay/Items/ScopedBurstSMG/Audio/SMG_RedDot_ReloadCartridge
root=Random_0
nodes [
 {id=Random_0, children [WavePlayer_0,WavePlayer_1,WavePlayer_2], properties {bShouldExcludeFromBranchCulling=true, weights [1,1,1]}}
 {id=WavePlayer_0, properties {soundWaveAssetPtr=&/SMG_RedDot_ReloadCartridge_01}}
 {id=WavePlayer_1, properties {soundWaveAssetPtr=&/SMG_RedDot_ReloadCartridge_02}}
 {id=WavePlayer_2, properties {soundWaveAssetPtr=&/SMG_RedDot_ReloadCartridge_03}}
]
properties {
 attenuationSettings=&Fort_Weapon_Reload_Att
 maxDistance=2250
 soundClassObject=&Fort_WeaponFoley
 volumeMultiplier=1.504
}
