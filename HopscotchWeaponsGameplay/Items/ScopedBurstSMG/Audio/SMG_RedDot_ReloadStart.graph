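# Sound cue for the start of the reload on the red-dot scoped burst SMG in Hopscotch weapons gameplay.
# A Random node picks one of three equally weighted reload-start waves, excluded from branch culling, and feeds a Modulator for variation.
# It plays in the weapon foley sound class with shared reload attenuation and a boosted volume.
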SoundCue /HopscotchWeaponsGameplay/Items/ScopedBurstSMG/Audio/SMG_RedDot_ReloadStart
root=Modulator_0
nodes [
 {id=Modulator_0, children [Random_1]}
 {id=Random_1, children [WavePlayer_3,WavePlayer_4,WavePlayer_5], properties {bShouldExcludeFromBranchCulling=true, weights [1,1,1]}}
 {id=WavePlayer_3, properties {soundWaveAssetPtr=&/SMG_RedDot_ReloadStart_03}}
 {id=WavePlayer_4, properties {soundWaveAssetPtr=&/SMG_RedDot_ReloadStart_01}}
 {id=WavePlayer_5, properties {soundWaveAssetPtr=&/SMG_RedDot_ReloadStart_02}}
]
properties {
 attenuationSettings=&Fort_Weapon_Reload_Att
 maxDistance=2250
 soundClassObject=&Fort_WeaponFoley
 volumeMultiplier=1.504
}
